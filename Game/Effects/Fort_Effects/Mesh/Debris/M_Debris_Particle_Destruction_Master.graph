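# Master lit material for destruction debris Niagara mesh particles, with dynamic parameters driving glow and burnt looks.
# Diffuse blends toward a particle-tinted desaturated version, and Burnt mixes in a darkened, desaturated char and pushes up roughness.
# A static switch picks colored Fresnel or a fiery curve-atlas ramp for emissive, scaled by Glow.
# The default shading path derives HDR emissive from LDR, while Forward and Mobile use the clamped value.
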Material /Game/Effects/Fort_Effects/Mesh/Debris/M_Debris_Particle_Destruction_Master
graph {
 properties {bUsedWithNiagaraMeshParticles=true, pixelDepthOffsetMode=PDOM_Legacy, usedShadingModels=Default Lit}
 nodes [
  n1 = NamedRerouteUsage()
  n2 = Constant() {r=0.5}
  LinearInterpolate(
   A=NamedRerouteUsage()
   B=Multiply(A=NamedRerouteUsage(), B=Desaturation(None=@n1, Fraction=@n2))
   Alpha=DynamicParameter() {defaultValue=<rgba 0 0 2 1>, paramNames [Glow, Burnt, GlowFresnelExp, SubUVSelection]}.Burnt
  )
  n3 = DynamicParameter() {defaultValue=<rgba 0 0 2 1>, paramNames [Glow, Burnt, GlowFresnelExp, SubUVSelection]}
  n4 = VertexNormalWS()
  n5 = Multiply(A=NamedRerouteUsage(), B=NamedRerouteUsage())
  n6 = ScalarParameter() {defaultValue=1, group=Glow, parameterName=FresnelExp, sortPriority=2}
  n7 = Saturate(None=Add(A=LinearInterpolate(Alpha=@n5), B=Fresnel(ExponentIn=@n6) {baseReflectFraction=0, exponent=1}))
  n8 = CurveAtlasRowParameter(CurveTime=@n7) {
   atlas=/Game/Curie/Effects/temp/Curves/CA_Curie_Fire_Visuals
   curve=/Game/Curie/Effects/temp/Curves/Curve_Fiery
   defaultValue=2
   group=Glow
   parameterName=Curie
  }
  n9 = StaticSwitchParameter(
   True=Multiply(
    A=VectorParameter() {parameterName=EmissiveFresnelColor, defaultValue=<rgba 0 0.4503 1 1>}.RGB
    B=Fresnel(ExponentIn=@n3.GlowFresnelExp, Normal=@n4) {baseReflectFraction=0, exponent=1}
   )
   False=Multiply(A=@n8, B=Add(A=@n7) {constB=0.1})
  ) {parameterName=EmissiveFresnel}
  n10 = Saturate(None=@n9)
  n11 = ScalarParameter() {defaultValue=1, group=Glow, parameterName=HDRIntensity, sortPriority=3}
  n12 = Multiply(
   A=ShadingPathSwitch(
    Default=MaterialFunctionCall(LDR Input=@n10, HDR Tint=@n9, HDR Intensity Multiplier=@n11) {materialFunction=&DeriveHDRfromLDR}.RemappedRange
    Forward=@n10
    Mobile=@n10
   )
   B=DynamicParameter() {defaultValue=<rgba 0 0 2 1>, paramNames [Glow, Burnt, GlowFresnelExp, SubUVSelection]}.Glow
  )
  MaterialFunctionCall(Emissive=@n12) {materialFunction=&MF_FX_EmissiveAdjustments}
  n13 = TextureSampleParameter2D() {parameterName=Diffuse, texture=&/T_Debris_Particle_D}
  n14 = Constant() {r=1}
  n15 = Saturate(None=Multiply(A=Desaturation(None=@n13.RGB, Fraction=@n14), B=ScalarParameter() {defaultValue=1, group=Diffuse, parameterName=Contrast, sortPriority=2}))
  n16 = ParticleColor()
  NamedRerouteDeclaration(None=LinearInterpolate(A=@n13.RGB, B=Multiply(A=@n15, B=@n16.RGB), Alpha=ScalarParameter() {group=Diffuse, parameterName=Use Particle Color Only, sortPriority=1})) {name=Diffuse, nodeColor=<rgba 0 0.6484 1 1>}
  n17 = TextureSample() {
   samplerType=SAMPLERTYPE_LinearGrayscale
   texture=/Game/Characters/Player/Male/Medium/Bodies/M_Med_Soldier_04/Skins/Gingerbread_Burnt/Textures/BurnGradient_output
  }
  NamedRerouteDeclaration(None=LinearInterpolate(Alpha=@n17.G) {constA=0.05, constB=0.9}) {name=Burnt Sample, nodeColor=<rgba 0.2422 1 0 1>}
  NamedRerouteDeclaration(None=TextureSampleParameter2D() {parameterName=Normal, samplerType=SAMPLERTYPE_Normal, texture=&/T_Debris_Particle_N}.B) {name=Glow Normal Read, nodeColor=<rgba 0.6172 1 0 1>}
  NamedRerouteDeclaration(None=@n12) {name=Emissive, nodeColor=<rgba 1 0 0.1875 1>}
  NamedRerouteUsage()
  n18 = TextureSampleParameter2D() {parameterName=Masks, texture=&/T_Debris_Particle_M}
  Saturate(None=Multiply(A=@n18.R, B=ScalarParameter() {defaultValue=1, parameterName=Metallic Intensity}))
  n19 = Add(A=@n18.G, B=DynamicParameter() {defaultValue=<rgba 0 0 2 1>, paramNames [Glow, Burnt, GlowFresnelExp, SubUVSelection]}.Burnt)
  Saturate(None=LinearInterpolate(
    A=ScalarParameter() {parameterName=Min Roughness}
    B=ScalarParameter() {defaultValue=1, parameterName=Max Roughness}
    Alpha=Saturate(None=@n19)
   ))
 ]
}
