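# Master Default Lit material for Niagara mesh debris particles that can look burnt and glow like embers.
# Dynamic parameters drive both states: Burnt blends particle-tinted diffuse toward partly desaturated, noise-darkened diffuse, and Glow scales the emissive.
# Glow blends two colors using burn noise, normal-map blue and fresnel, with HDR remapping only on the default shading path.
# Scalar parameters allow grayscale particle-color tinting, base color as emissive and normal flattening.
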Material /Game/Effects/Fort_Effects/Mesh/Debris/M_Debris_Particle_Master
graph {
 properties {bUsedWithNanite=true, bUsedWithNiagaraMeshParticles=true, pixelDepthOffsetMode=PDOM_Legacy, usedShadingModels=Default Lit}
 nodes [
  n1 = DynamicParameter() {defaultValue=<rgba 0 0 1 1>, paramNames [Glow,Burnt,Param3,Param4]}
  Add(A=@n1.Burnt, B=TextureSampleParameter2D() {parameterName=Masks, texture=&/T_Debris_Particle_M}.G)
  n2 = TextureSample() {
   samplerType=SAMPLERTYPE_LinearGrayscale
   texture=/Game/Characters/Player/Male/Medium/Bodies/M_Med_Soldier_04/Skins/Gingerbread_Burnt/Textures/BurnGradient_output
  }
  n3 = LinearInterpolate(
   A=ScalarParameter() {defaultValue=0.05, parameterName=BurnNoiseMin}
   B=ScalarParameter() {defaultValue=0.9, parameterName=BurnNoiseMax}
   Alpha=@n2.G
  ) {constA=0.05, constB=0.9}
  n4 = TextureSampleParameter2D() {parameterName=Normal, samplerType=SAMPLERTYPE_Normal, texture=&/T_Debris_Particle_N}
  n5 = Multiply(A=@n3, B=@n4.B)
  n6 = ScalarParameter() {defaultValue=2.5, parameterName=FresnelExp}
  n7 = Saturate(None=Add(A=LinearInterpolate(Alpha=@n5) {constA=0.5, constB=-0.4}, B=Fresnel(ExponentIn=@n6) {baseReflectFraction=0, exponent=1}))
  n8 = Multiply(
   A=LinearInterpolate(
    A=VectorParameter() {group=Color, parameterName=GlowColor2, defaultValue=<rgba 1 0.01 0 0>}.RGB
    B=VectorParameter() {group=Color, parameterName=GlowColor1, defaultValue=<rgba 1 0.2437 0 0>}.RGB
    Alpha=@n7
   )
   B=Add(A=@n7) {constB=0.1}
  )
  n9 = Clamp(None=@n8)
  n10 = ShadingPathSwitch(
   Default=MaterialFunctionCall(LDR Input=@n9, HDR Tint=@n8, HDR Intensity Multiplier=ScalarParameter() {defaultValue=1, parameterName=HDRIntensity}) {materialFunction=&DeriveHDRfromLDR}.RemappedRange
   Forward=@n9
   Mobile=@n9
  )
  n11 = ScalarParameter() {defaultValue=1, parameterName=Base Color Emissive Intensity}
  n12 = TextureSampleParameter2D() {parameterName=Diffuse, texture=&/T_Debris_Particle_D}
  n13 = Desaturation(None=@n12.RGB, Fraction=Constant() {r=1})
  n14 = ScalarParameter() {defaultValue=1, parameterName=Contrast}
  n15 = LinearInterpolate(A=@n12.RGB, B=Saturate(None=Multiply(A=@n13, B=@n14)), Alpha=ScalarParameter() {parameterName=Use Particle Color Only})
  n16 = ParticleColor()
  n17 = Constant() {r=0.5}
  n18 = LinearInterpolate(
   A=Multiply(A=Multiply(A=@n15, B=@n16.RGB), B=VectorParameter() {parameterName=BaseColor, defaultValue=<rgba 1 1 1 0>}.RGB)
   B=Multiply(A=@n3, B=Desaturation(None=@n12.RGB, Fraction=@n17))
   Alpha=@n1.Burnt
  )
  MaterialFunctionCall(Emissive=LinearInterpolate(A=Multiply(A=@n10, B=@n1.Glow), B=Multiply(A=@n11, B=@n18), Alpha=ScalarParameter() {parameterName=Use Base Color as Emissive})) {materialFunction=&MF_FX_EmissiveAdjustments}
  MaterialFunctionCall(Normal=@n4.RGB, Flatness=ScalarParameter() {defaultValue=0.25, parameterName=NormalFlatness}) {materialFunction=&FlattenNormal}
 ]
}
