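# Masked, lit Niagara mesh-particle material that turns mesh pieces into sprites of flipbook debris, using rock textures by default.
# Vertex colors randomize each piece's size, brightness, spin and flipbook frame or optional animation, while dynamic parameters drive scale, added spin and emissive glow.
# UVs rotate around the current cell's center, with sampled normals counter-rotated to match.
# Sprites shrink near the camera, and a Fresnel term shapes the glow.
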Material /Game/Effects/Fort_Effects/Mesh/Debris/M_FlipbookDebris_MeshSprite
graph {
 properties {bAutomaticallySetUsageInEditor=false, bUsedWithNiagaraMeshParticles=true, blendMode=BLEND_Masked, usedShadingModels=Default Lit}
 nodes [
  n1 = VertexColor()
  n2 = ScalarParameter() {defaultValue=2, parameterName=SizeVariancePower}
  n3 = Multiply(
   A=LinearInterpolate(A=ScalarParameter() {defaultValue=0.07, parameterName=MinSpriteSizeMult}, Alpha=Power(Base=@n1.B, Exp=@n2) {constExponent=2.5}) {constA=1}
   B=ScalarParameter() {defaultValue=50, parameterName=SpriteSize}
  )
  n4 = DynamicParameter() {defaultValue=<rgba 0.5 1 1 0>, paramNames [SizeScale 0-1, SizeScaleBase, SpinAdd, EmissiveGlow]}
  n5 = Multiply(A=Saturate(None=Subtract(A=@n4.SizeScale 0-1) {constB=0.75})) {constB=4}
  n6 = Multiply(B=@n4.SizeScale 0-1) {constA=2, constB=2}
  n7 = Multiply(A=@n4.SizeScaleBase, B=Saturate(None=LinearInterpolate(A=@n5, B=@n6, Alpha=@n1.B) {constA=12, constB=0.6}))
  n8 = ScalarParameter() {defaultValue=500, parameterName=NearCameraShrinkDist}
  n9 = Time()
  n10 = VertexColor()
  n11 = ScalarParameter() {defaultValue=0.6, parameterName=FlipbookAnimSpeedMin}
  n12 = ScalarParameter() {defaultValue=1.2, parameterName=FlipbookAnimSpeedMax}
  n13 = VertexColor()
  n14 = StaticSwitchParameter(True=Multiply(A=Add(A=@n9, B=@n10.G), B=LinearInterpolate(A=@n11, B=@n12, Alpha=@n10.R)), False=@n13.R) {parameterName=UseFlipbookAnimation}
  n15 = ScalarParameter() {defaultValue=4, parameterName=SubUVFrames}
  n16 = Multiply(A=@n15, B=@n15)
  n17 = Frac(None=@n14)
  n18 = Floor(None=Multiply(A=AppendVector(A=@n16, B=@n15), B=AppendVector(A=@n17, B=@n17)))
  n19 = Multiply(A=Add(A=Time(), B=@n13.R), B=ScalarParameter() {defaultValue=0.1, parameterName=BaseSpinSpeed})
  n20 = Multiply(
   A=NamedRerouteDeclaration(None=Add(A=@n19, B=@n4.SpinAdd)) {name=Animation, nodeColor=<rgba 1 0 0.0234 1>}
   B=LinearInterpolate(Alpha=@n13.G) {constA=-0.5, constB=0.5}
  )
  n21 = MaterialFunctionCall(
   UVs=MaterialFunctionCall("Animation  Phase (0-1)"=@n14, Number of Rows=@n15, Number of Columns=@n15, UVs=TextureCoordinate()) {materialFunction=&FlipBook}.UVs
   Rotation Center=Divide(A=Add(A=@n18) {constB=0.5}, B=AppendVector(A=@n15, B=@n15))
   "Rotation Angle (0-1)"=@n20
  ) {materialFunction=&CustomRotator}
  n22 = TextureSampleParameter2D(UVs=@n21.Rotated Values) {parameterName=Normal, samplerType=SAMPLERTYPE_Normal, texture=/Game/Effects/Fort_Effects/Textures/Rocks/T_Rock_Stone_Sprite_N}
  n23 = Constant3Vector() {constant=<rgba 0 0 1 0>}
  n24 = Multiply(A=@n20) {constB=-1}
  n25 = Constant()
  MaterialFunctionCall!(
   Center Location=WorldPosition().XYZ
   0-1 UVs
   XY scale=Multiply(
    A=Multiply(A=@n3, B=@n7)
    B=MaterialFunctionCall(Distance=@n8) {materialFunction=&Near_Camera_Fade}.Result
   )
   OverridePivot
   "Normals (Optional)"=Add(A=@n22.RGB, B=RotateAboutAxis(NormalizedRotationAxis=@n23, RotationAngle=@n24, PivotPoint=@n25, Position=@n22.RGB))
  ) {materialFunction=&Sprite}
  n26 = Fresnel(ExponentIn=ScalarParameter() {defaultValue=2, parameterName=EmissiveFresnelExp}) {baseReflectFraction=0.1, exponent=2}
  n27 = VectorParameter() {parameterName=EmissiveColor}
  MaterialFunctionCall(Emissive=Multiply(A=NamedRerouteUsage(), B=Multiply(A=@n26, B=@n27.RGB))) {materialFunction=&MF_FX_EmissiveAdjustments}
  n28 = VertexColor()
  n29 = TextureSampleParameter2D(UVs=@n21.Rotated Values) {parameterName=BaseColorMask, texture=/Game/Effects/Fort_Effects/Textures/Rocks/T_Rock_Stone_Sprite_D}
  n30 = LinearInterpolate(A=ScalarParameter() {defaultValue=0.5, parameterName=MinDiffuseBrightness}, Alpha=@n29.G) {constA=0.5}
  n31 = ScalarParameter() {parameterName=UseDiffuseColor}
  Multiply(A=ParticleColor().RGB, B=Multiply(A=LinearInterpolate(Alpha=@n28.R) {constA=0.6}, B=LinearInterpolate(A=@n30, B=@n29.RGB, Alpha=@n31)))
  Multiply(A=SphereMask(A=TextureCoordinate(), B=Constant() {r=0.5}) {attenuationRadius=0.52}, B=@n29.A)
  NamedRerouteDeclaration(None=@n4.EmissiveGlow) {name=EmissiveGlow, nodeColor=<rgba 0 1 0.4609 1>}
 ]
}
