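# Default Lit material for metal debris chunk mesh particles in Cascade and Niagara effects.
# A mask blends PaintColor-tinted metal with darkened diffuse, alongside a normal map.
# Squared specular-map blue plus Fresnel drives a two-color glow, multiplied by particle color and gated by the UseGlow static switch.
# The default shading path derives HDR glow from LDR, mobile uses clamped color, and the result passes through MF_FX_EmissiveAdjustments.
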Material /Game/Effects/Fort_Effects/Mesh/Debris/M_Metal_Frag_Chunk_FX
graph {
 properties {
  bUsedWithMeshParticles=true
  bUsedWithNiagaraMeshParticles=true
  pixelDepthOffsetMode=PDOM_Legacy
  previewMesh=&/SM_Metal_Debris_Chunk_1
  usedShadingModels=Default Lit
 }
 nodes [
  n1 = TextureSampleParameter2D() {parameterName=Diffuse, texture=/Game/Packages/PBW/Metal/Textures/T_Metal_L1}
  LinearInterpolate(
   A=Multiply(A=VectorParameter() {parameterName=PaintColor, defaultValue=<rgba 0.66 0.8802 0.8715 0>}.RGB, B=@n1.B)
   B=Multiply(A=@n1.RGB) {constB=0.6}
   Alpha=TextureSample() {texture=/Game/Packages/PBW/Metal/Textures/T_Metal_L1_WSCP_M}.G
  )
  n2 = TextureSample() {texture=/Game/Packages/PBW/Metal/Textures/T_Metal_L1_Spec}
  n3 = Multiply(A=@n2.B, B=@n2.B)
  n4 = ScalarParameter() {defaultValue=3, parameterName=FresnelExp}
  n5 = Saturate(None=Add(A=LinearInterpolate(Alpha=@n3) {constA=0.02, constB=0.4}, B=Fresnel(ExponentIn=@n4) {baseReflectFraction=0, exponent=1}))
  n6 = Multiply(
   A=LinearInterpolate(
    A=VectorParameter() {group=Color, parameterName=GlowColor2, defaultValue=<rgba 1 0.01 0 0>}.RGB
    B=VectorParameter() {group=Color, parameterName=GlowColor1, defaultValue=<rgba 1 0.4 0.0111 0>}.RGB
    Alpha=@n5
   )
   B=Add(A=@n5) {constB=0.2}
  )
  n7 = Clamp(None=@n6)
  n8 = ShadingPathSwitch(
   Default=MaterialFunctionCall(LDR Input=@n7, HDR Tint=@n6, HDR Intensity Multiplier=ScalarParameter() {defaultValue=1, parameterName=HDRIntensity}) {materialFunction=&DeriveHDRfromLDR}.RemappedRange
   Mobile=@n7
  ) {desc=Migrated from FeatureLevelSwitch}
  n9 = ParticleColor()
  MaterialFunctionCall(Emissive=StaticSwitchParameter(True=Multiply(A=@n8, B=@n9.RGB), False=Constant()) {parameterName=UseGlow}) {materialFunction=&MF_FX_EmissiveAdjustments}
  TextureSampleParameter2D() {parameterName=Normal, samplerType=SAMPLERTYPE_Normal, texture=/Game/Packages/PBW/Metal/Textures/T_Metal_L1_N}
 ]
}
