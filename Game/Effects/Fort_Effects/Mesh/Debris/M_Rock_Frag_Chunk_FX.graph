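# Lit material for rock fragment debris on Cascade and Niagara mesh particles, with optional glowing cracks.
# A tiled crack texture, scaled by the Cracks dynamic parameter, alters base color and roughness.
# With UseGlow, normal-map blue, cracks and Fresnel blend two glow colors into HDR-remapped emissive tinted by particle color; otherwise particle color tints base color.
# A second dynamic parameter feeds a previous-frame switch, scaled for motion blur.
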Material /Game/Effects/Fort_Effects/Mesh/Debris/M_Rock_Frag_Chunk_FX
graph {
 properties {
  bUsedWithMeshParticles=true
  bUsedWithNiagaraMeshParticles=true
  pixelDepthOffsetMode=PDOM_Legacy
  previewMesh=&/SM_Rock_Frag_Chunk_FX
  usedShadingModels=Default Lit
 }
 nodes [
  n1 = ScalarParameter() {defaultValue=0.6, parameterName=Roughness}
  n2 = TextureSample(UVs=TextureCoordinate() {uTiling=2, vTiling=2}) {texture=/Game/Global/Textures/Default/Masks_and_Gradients/Cracks_02}
  n3 = DynamicParameter() {defaultValue=<rgba 0.5 1 1 1>, paramNames [Cracks,Param2,Param3,Param4]}
  LinearInterpolate(
   A=@n1
   B=LinearInterpolate(A=@n1, B=ScalarParameter() {defaultValue=0.97, parameterName=CrackRoughness}, Alpha=@n2.B) {constA=1}
   Alpha=@n3.Cracks
  )
  n4 = LinearInterpolate(Alpha=TextureSampleParameter2D() {parameterName=Normal, samplerType=SAMPLERTYPE_Normal, texture=/Game/Environments/Props/Rocks/Textures/T_Props_Rocks_N}.B) {constA=0.5, constB=0}
  n5 = Multiply(A=@n2.R, B=@n3.Cracks)
  n6 = ScalarParameter() {defaultValue=2, parameterName=FresnelExp}
  n7 = Saturate(None=Add(A=Add(A=@n4, B=@n5), B=Fresnel(ExponentIn=@n6) {baseReflectFraction=0, exponent=1}))
  n8 = Multiply(
   A=LinearInterpolate(
    A=VectorParameter() {group=Color, parameterName=GlowColor2, defaultValue=<rgba 1 0.01 0 0>}.RGB
    B=VectorParameter() {group=Color, parameterName=GlowColor1, defaultValue=<rgba 1 0.2437 0 0>}.RGB
    Alpha=@n7
   )
   B=Add(A=@n7) {constB=0.2}
  )
  n9 = Clamp(None=@n8)
  n10 = ShadingPathSwitch(
   Default=MaterialFunctionCall(LDR Input=@n9, HDR Tint=@n8, HDR Intensity Multiplier=ScalarParameter() {defaultValue=1, parameterName=HDRIntensity}) {materialFunction=&DeriveHDRfromLDR}.RemappedRange
   Mobile=@n9
  ) {desc=Migrated from FeatureLevelSwitch}
  n11 = ParticleColor()
  MaterialFunctionCall(Emissive=StaticSwitchParameter(True=Multiply(A=@n10, B=@n11.RGB), False=Constant()) {parameterName=UseGlow}) {materialFunction=&MF_FX_EmissiveAdjustments}
  n12 = DynamicParameter() {parameterIndex=1, defaultValue=<rgba 0 0 0 1>, paramNames [BlurX,BlurY,BlurZ,Param4]}
  Multiply(
   A=PreviousFrameSwitch(Previous Frame=MaterialFunctionCall(R=@n12.BlurX, G=@n12.BlurX, B=@n12.BlurX) {materialFunction=&AppendMany}.RGB)
   B=ScalarParameter() {parameterName=DynParamMotionBlur}
  )
  n13 = TextureSampleParameter2D() {parameterName=Diffuse, texture=/Game/Environments/Props/Rocks/Textures/T_Props_Rocks_D}
  n14 = Multiply(A=@n13.RGB, B=@n13.RGB)
  n15 = ScalarParameter() {defaultValue=0.02, parameterName=CrackBrightness}
  n16 = Multiply(
   A=VectorParameter() {parameterName=BaseColorTint, defaultValue=<rgba 0.25 0.1617 0.1443 1>}.RGB
   B=LinearInterpolate(A=@n14, B=LinearInterpolate(A=@n14, B=@n15, Alpha=@n2.R), Alpha=@n3.Cracks)
  )
  StaticSwitchParameter(True=@n16, False=Multiply(A=@n11.RGB, B=@n16)) {parameterName=UseGlow}
 ]
}
